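; Anim notify state that plays a configurable camera animation on the owning Fort pawn's player controller for the duration of the notify window.
; Begin casts to the pawn and controller, optionally ignores look input, then plays the animation through the camera manager's CameraAnimationCameraModifier using play rate, scale, blend, loop and play-space variables.
; End re-enables look input if it was disabled and stops all instances of that animation.
; Camera calls run only for local controllers.
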(fn Received_NotifyBegin
  (params (MeshComp :type Skeletal%20Mesh%20Component%20Object%20Reference) (Animation :type Anim%20Sequence%20Base%20Object%20Reference) (TotalDuration :type Float%20%28single-precision%29) (EventReference :type Anim%20Notify%20Event%20Reference%20Structure%20%28by%20ref%29))
  (bind _asfort_pawn (Utilities|Casting|CastToFortPawn (Components|GetOwner MeshComp))
    (:then
      (bind _asfort_player_controller (Utilities|Casting|CastToFortPlayerController (Pawn|GetController _asfort_pawn))
        (:then
          (if (Variables|Default|GetbIgnoreLookInput)
            (Input|SetIgnoreLookInput _asfort_player_controller true)
            (else
              _))
          (bind _playercameramanager (Variables|/Script/Engine.PlayerController|GetPlayerCameraManager _asfort_player_controller)
            (:then
              (if (Pawn|IsLocalController _asfort_player_controller)
                (CameraAnimation|PlayCameraAnimation (CameraAnimation|ConvCameraAnimationCameraModifier _playercameramanager) (Variables|Default|GetCameraAnim) (Utilities|Struct|MakeCameraAnimationParams (Variables|Default|GetPlayRate) (Variables|Default|GetShake%20Scale) "Linear" (Variables|Default|GetBlendInTime) "Linear" (Variables|Default|GetBlendOutTime) (Variables|Default|GetbLoop) 0 false 0.0 (CameraAnimation|ConvCameraAnimationPlaySpace (Variables|Default|GetShake%20Space))) :!class /Script/EngineCameras.CameraAnimationCameraModifier)
                (else
                  _))
              (return false))
            (:else
              _)))
        (:CastFailed
          _)))
    (:CastFailed
      _)))

(fn Received_NotifyEnd
  (params (MeshComp :type Skeletal%20Mesh%20Component%20Object%20Reference) (Animation :type Anim%20Sequence%20Base%20Object%20Reference) (EventReference :type Anim%20Notify%20Event%20Reference%20Structure%20%28by%20ref%29))
  (bind _asfort_pawn (Utilities|Casting|CastToFortPawn (Components|GetOwner MeshComp))
    (:then
      (bind _asfort_player_controller (Utilities|Casting|CastToFortPlayerController (Pawn|GetController _asfort_pawn))
        (:then
          (if (Variables|Default|GetbIgnoreLookInput)
            (Input|SetIgnoreLookInput _asfort_player_controller false)
            (else
              _))
          (bind _playercameramanager (Variables|/Script/Engine.PlayerController|GetPlayerCameraManager _asfort_player_controller)
            (:then
              (if (Pawn|IsLocalController _asfort_player_controller)
                (CameraAnimation|StopAllCameraAnimationsOf (CameraAnimation|ConvCameraAnimationCameraModifier _playercameramanager) (Variables|Default|GetCameraAnim) false :!class /Script/EngineCameras.CameraAnimationCameraModifier)
                (else
                  _))
              (return false))
            (:else
              _)))
        (:CastFailed
          _)))
    (:CastFailed
      _)))
