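; Anim notify state that plays a camera shake for the animated pawn's player over the notify's duration.
; Both begin and end do nothing on a dedicated server, otherwise casting the mesh owner to FortPawn and its controller to FortPlayerController.
; Begin starts the ShakeBP variable's shake at ShakeScale in camera-local space, and end stops that same shake immediately.
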(fn Received_NotifyBegin
  (params (MeshComp :type Skeletal%20Mesh%20Component%20Object%20Reference) (Animation :type Anim%20Sequence%20Base%20Object%20Reference) (TotalDuration :type Float%20%28single-precision%29) (EventReference :type Anim%20Notify%20Event%20Reference%20Structure%20%28by%20ref%29))
  (if (Networking|IsDedicatedServer MeshComp)
    (return false)
    (else
      (bind _asfort_pawn (Utilities|Casting|CastToFortPawn (Components|GetOwner MeshComp))) ; confirm fort pawn
      (bind _asfort_player_controller (Utilities|Casting|CastToFortPlayerController (Pawn|GetController _asfort_pawn))) ; Confirm it is a FortPlayerControler
      (Game|Feedback|ClientStartCameraShake :self _asfort_player_controller :Shake (Variables|Default|GetShakeBP) :Scale (Variables|Default|GetShakeScale) :PlaySpace "CameraLocal") ; start the shake
    )))

(fn Received_NotifyEnd
  (params (MeshComp :type Skeletal%20Mesh%20Component%20Object%20Reference) (Animation :type Anim%20Sequence%20Base%20Object%20Reference) (EventReference :type Anim%20Notify%20Event%20Reference%20Structure%20%28by%20ref%29))
  (if (Networking|IsDedicatedServer MeshComp)
    (return false)
    (else
      (bind _asfort_pawn (Utilities|Casting|CastToFortPawn (Components|GetOwner MeshComp))) ; confirm player
      (bind _asfort_player_controller (Utilities|Casting|CastToFortPlayerController (Pawn|GetController _asfort_pawn))) ; confirm player controller
      (Game|Feedback|ClientStopCameraShake _asfort_player_controller (Variables|Default|GetShakeBP) true) ; stop the shake
    )))
