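; Anim notify that plays a configurable camera animation sequence for the player controlling the animated mesh's owning pawn.
; On notify, it casts the owner to FortPawn and its controller to FortPlayerController, then gets the camera animation modifier from the camera manager.
; If the modifier and sequence are valid, it plays the sequence with the configured params.
; The notify's name shows the sequence's display name, falling back to the parent's name when unset.
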(fn GetNotifyName
  (bind _returnvalue (|CallParentFunction))
  (bind _cameraanimationsequence (Variables|Default|GetCameraAnimationSequence)
    (:then
      (return (Utilities|GetDisplayName _cameraanimationsequence)))
    (:else
      (return _returnvalue))))

(fn Received_Notify
  (params (MeshComp :type Skeletal%20Mesh%20Component%20Object%20Reference) (Animation :type Anim%20Sequence%20Base%20Object%20Reference) (EventReference :type Anim%20Notify%20Event%20Reference%20Structure%20%28by%20ref%29))
  ;[0 0.15 0.07299 0.5] ReceivedNotify: Play the intended camera animation.
  (bind _returnvalue (|CallParentFunction MeshComp Animation EventReference))
  (Utilities|FlowControl|Sequence
    (:then_0
      (bind _asfort_pawn (Utilities|Casting|CastToFortPawn (Components|GetOwner MeshComp)))
      (bind _asfort_player_controller (Utilities|Casting|CastToFortPlayerController (Pawn|GetController _asfort_pawn)))
      (bind _playercameramanager (Variables|/Script/Engine.PlayerController|GetPlayerCameraManager _asfort_player_controller)
        (:then
          (bind _returnvalue_2 (CameraAnimation|ConvCameraAnimationCameraModifier _playercameramanager))
          (Utilities|IsValid _returnvalue_2
            (:Is%20Valid
              (bind _cameraanimationsequence (Variables|Default|GetCameraAnimationSequence)
                (:then
                  (CameraAnimation|PlayCameraAnimation _returnvalue_2 _cameraanimationsequence (Variables|Default|GetCameraAnimationParams) :!class /Script/EngineCameras.CameraAnimationCameraModifier))
                (:else)))))
        (:else)))
    (:then_1
      (return _returnvalue))))
