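; BP_AnimNotify_CameraShake is an animation notify that plays a camera shake for the player controlling the animated pawn.
; On non-dedicated-server instances, it casts the mesh component's owner to FortPawn and that pawn's controller to FortPlayerController.
; It then calls ClientStartCameraShake on that controller using the notify's configurable shake class, scale and play space variables.
; Owners that are not player-controlled Fort pawns get no shake.
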(fn Received_Notify
  (params (MeshComp :type Skeletal%20Mesh%20Component%20Object%20Reference) (Animation :type Anim%20Sequence%20Base%20Object%20Reference) (EventReference :type Anim%20Notify%20Event%20Reference%20Structure%20%28by%20ref%29))
  (if (not (Networking|IsDedicatedServer))
    (bind _asfort_pawn (Utilities|Casting|CastToFortPawn (Components|GetOwner MeshComp))
      (:then
        (bind _asfort_player_controller (Utilities|Casting|CastToFortPlayerController (Pawn|GetController _asfort_pawn))
          (:then
            (Game|Feedback|ClientStartCameraShake :self _asfort_player_controller :Shake (Variables|Default|GetShake%20BP) :Scale (Variables|Default|GetShake%20Scale) :PlaySpace (Variables|Default|GetShake%20Space)))
          (:CastFailed
            _)))
      (:CastFailed
        _))
    (return false)))
