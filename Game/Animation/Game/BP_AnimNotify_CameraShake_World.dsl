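; Anim notify that plays a world camera shake from the animated mesh, skipped on dedicated servers.
; The epicenter is the configured socket's location, or the mesh component's world location when the socket name is empty.
; Shake class, inner and outer radius, and orientation toward the epicenter come from notify variables, while falloff is fixed.
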(fn Received_Notify
  (params (MeshComp :type Skeletal%20Mesh%20Component%20Object%20Reference) (Animation :type Anim%20Sequence%20Base%20Object%20Reference) (EventReference :type Anim%20Notify%20Event%20Reference%20Structure%20%28by%20ref%29))
  (if (not (Networking|IsDedicatedServer))
    (Camera|PlayWorldCameraShake (Variables|Default|GetShake%20BP) (Math|Vector|SelectVector (Transformation|GetWorldLocation MeshComp) (Transformation|GetSocketLocation MeshComp (Variables|Default|GetSocketName)) (== (Variables|Default|GetSocketName) "")) (Variables|Default|GetInnerRadius) (Variables|Default|GetOuterRadius) 1.0 (Variables|Default|GetOrientShakeTowardsEpicenter))
    (return false)))
